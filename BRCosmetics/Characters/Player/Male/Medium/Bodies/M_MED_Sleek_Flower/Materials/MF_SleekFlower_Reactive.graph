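# Material function for the Sleek Flower player body that reveals flower diffuse and a glow as eliminations accumulate.
# Blueprint-timestamped elimination progress, optionally quantized or overridden by parameter, samples a curve atlas to threshold a gradient mask.
# That threshold, combined with five stage-toggled sphere masks on an add-mask, blends in the reactive diffuse.
# A timed, sine-faded emissive band travels along the gradient, with both effects limited by an FX channel mask.
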MaterialFunction /BRCosmetics/Characters/Player/Male/Medium/Bodies/M_MED_Sleek_Flower/Materials/MF_SleekFlower_Reactive
graph {nodes [
  n1 = Time()
  n2 = ScalarParameter() {defaultValue=0.05, parameterName=Speed}
  ConstantBiasScale(None=Sine(None=Multiply(A=@n1, B=@n2)))
  Constant()
  n3 = TextureSampleParameter2D() {
   parameterName=SF_ReactiveGradient
   samplerType=SAMPLERTYPE_LinearGrayscale
   texture=/BRCosmetics/Characters/Player/Male/Medium/Bodies/M_MED_Sleek_Flower/Textures/T_Sleek_Flower_GRADNEW
  }
  n4 = ScalarParameter() {group=SuperLevelling_Mask, parameterName=SF_CurrentElims, sortPriority=3}
  n5 = Subtract(A=Time(), B=ScalarParameter() {defaultValue=-99999, group=SuperLevelling_Mask, parameterName=SF_GameTimeFromBP, sortPriority=100})
  n6 = Add(A=ScalarParameter() {defaultValue=1, group=SuperLevelling_Mask, parameterName=SF_Animation_Duration, sortPriority=101}) {constB=0}
  n7 = LinearInterpolate(A=Subtract(A=@n4), B=@n4, Alpha=Saturate(None=Divide(A=@n5, B=@n6)))
  n8 = ScalarParameter() {group=SuperLevelling_Mask, parameterName=SF_StartElims, sortPriority=3}
  n9 = ScalarParameter() {defaultValue=5, group=SuperLevelling_Mask, parameterName=SF_EndElims, sortPriority=3}
  n10 = Divide(A=Subtract(A=Clamp(None=@n7, Min=@n8, Max=@n9), B=@n8), B=Subtract(A=@n9, B=@n8))
  n11 = Max(A=Ceil(None=ScalarParameter() {defaultValue=1, group=SuperLevelling_Mask, parameterName=SL_QuantizationSteps, sliderMax=100, sliderMin=1}))
  n12 = Divide(A=Floor(None=Multiply(A=@n10, B=@n11)), B=@n11)
  n13 = Saturate(None=Ceil(None=Subtract(A=@n11)))
  n14 = ScalarParameter() {parameterName=UseParameterValueForFlowers}
  n15 = CurveAtlasRowParameter(CurveTime=LinearInterpolate(A=LinearInterpolate(A=@n10, B=@n12, Alpha=@n13), B=ScalarParameter() {defaultValue=1, parameterName=TransformControl}, Alpha=@n14)) {
   atlas=/BRCosmetics/Characters/Player/Male/Medium/Bodies/M_MED_Sleek_Flower/Textures/CA_SleekFlower_RE
   curve=/BRCosmetics/Characters/Player/Male/Medium/Bodies/M_MED_Sleek_Flower/Textures/FlowerTransitionCurve_RE
   parameterName=TransitionCurve
  }
  n16 = Subtract(A=Add(A=@n3.G), B=Multiply(A=@n15.G) {constB=2})
  CurveAtlasRowParameter(CurveTime=Clamp(None=@n16)) {
   atlas=&CA_SleekFlower_RE
   curve=&FlowerTransitionCurve_RE
   parameterName=TransitionCurve
  }
  n17 = Subtract(A=Time(), B=ScalarParameter() {defaultValue=-99999, group=SuperLevelling_Mask, parameterName=SF_GameTimeFromBPEmissive, sortPriority=100})
  n18 = Add(A=ScalarParameter() {defaultValue=2, group=SuperLevelling_Mask, parameterName=SF_EmissiveBLink_Duration, sortPriority=101}) {constB=0}
  n19 = ScalarParameter() {parameterName=SF_FrontEnd_GlowTIme}
  n20 = LinearInterpolate(A=Saturate(None=Divide(A=@n17, B=@n18)), B=Frac(None=Multiply(A=@n19) {constB=5}), Alpha=@n14)
  CurveAtlasRowParameter(CurveTime=@n20) {
   atlas=&CA_SleekFlower_RE
   curve=&FlowerTransitionCurve_RE
   parameterName=TransitionCurve
  }
  n21 = GetMaterialAttributes(None=FunctionInput() {inputName=Material, inputType=FunctionInput_MaterialAttributes}) {attributeGetTypes [#1, #2]}
  n22 = TextureSampleParameter2D() {
   parameterName=SF_ReactiveDiffuse
   texture=/BRCosmetics/Characters/Player/Male/Medium/Bodies/M_MED_Sleek_Flower/Textures/T_M_MED_Sleek_Glasses_Flower_D
  }
  n23 = Ceil(None=@n16)
  n24 = TextureSample() {
   samplerType=SAMPLERTYPE_LinearGrayscale
   texture=/BRCosmetics/Characters/Player/Male/Medium/Bodies/M_MED_Sleek_Flower/Textures/T_Sleek_Glasses_Flower_ADDMASK
  }
  n25 = ScalarParameter() {defaultValue=1, parameterName=Stage_01_Value}
  n26 = ScalarParameter() {defaultValue=0.1, parameterName=Radius}
  n27 = Constant() {r=1}
  n28 = Multiply(
   A=SphereMask(
    A=@n24.G
    B=ScalarParameter() {defaultValue=0.75, parameterName=Stage_02_Value}
    Radius=ScalarParameter() {defaultValue=0.1, parameterName=Radius}
    Hardness=Constant() {r=1}
   )
   B=ScalarParameter() {parameterName=SF_Stage_02}
  )
  n29 = Multiply(
   A=SphereMask(
    A=@n24.G
    B=ScalarParameter() {defaultValue=0.5, parameterName=Stage_03_Value}
    Radius=ScalarParameter() {defaultValue=0.1, parameterName=Radius}
    Hardness=Constant() {r=1}
   )
   B=ScalarParameter() {parameterName=SF_Stage_03}
  )
  n30 = Multiply(
   A=SphereMask(
    A=@n24.G
    B=ScalarParameter() {defaultValue=0.25, parameterName=Stage_04_Value}
    Radius=ScalarParameter() {defaultValue=0.1, parameterName=Radius}
    Hardness=Constant() {r=1}
   )
   B=ScalarParameter() {parameterName=SF_Stage_04}
  )
  n31 = Multiply(
   A=SphereMask(
    A=@n24.G
    B=ScalarParameter() {parameterName=Stage_05_Value}
    Radius=ScalarParameter() {defaultValue=0.1, parameterName=Radius}
    Hardness=Constant() {r=1}
   )
   B=ScalarParameter() {parameterName=SF_Stage_05}
  )
  n32 = Max(
   A=Multiply(A=SphereMask(A=@n24.G, B=@n25, Radius=@n26, Hardness=@n27), B=ScalarParameter() {parameterName=SF_Stage_01})
   B=Max(A=Max(A=@n28, B=@n29), B=Max(A=@n30, B=@n31))
  )
  n33 = LinearInterpolate(A=@n21.BaseColor, B=@n22.RGB, Alpha=Multiply(A=Clamp(None=@n23), B=Clamp(None=@n32)))
  n34 = TextureSample() {samplerType=SAMPLERTYPE_Masks, texture=/BRCosmetics/Characters/Player/Male/Medium/Bodies/M_MED_Sleek_Flower/Textures/T_M_MED_Sleek_Flower_FX}
  n35 = ChannelMaskParameter(None=FunctionInput(Preview=@n34.RGBA) {inputName=FX_Mask, inputType=FunctionInput_Vector4}) {parameterName=SF_TransitionMask_Channel}
  n36 = ScalarParameter() {parameterName=SF_SwitchtoLoopingEmissive}
  n37 = OneMinus(None=LinearInterpolate(A=LinearInterpolate(A=@n15.B, Alpha=@n36) {constB=0}, Alpha=@n20) {constA=1})
  n38 = ScalarParameter() {defaultValue=0.15, parameterName=EmissiveRadius}
  n39 = Constant()
  n40 = TextureSample() {
   samplerType=SAMPLERTYPE_LinearColor
   texture=/BRCosmetics/Characters/Player/Male/Medium/Bodies/M_MED_Sleek_Flower/Textures/T_SleekFlowerEmissiveFX
  }
  n41 = VectorParameter() {parameterName=SF_EmissiveColor, defaultValue=<rgba 1 0.3618 0 1>}
  n42 = Power(Base=Multiply(A=Multiply(A=@n40.R, B=@n41.RGB), B=ScalarParameter() {defaultValue=5, parameterName=EmissiveStrength}))
  n43 = ConstantBiasScale(None=Sine(None=Add(A=@n20) {constB=-0.25}))
  n44 = Add(A=@n21.EmissiveColor, B=Multiply(A=SphereMask(A=@n3.G, B=@n37, Radius=@n38, Hardness=@n39), B=Multiply(A=@n42, B=@n43)))
  FunctionOutput(None=SetMaterialAttributes!(
    MaterialAttributes=@n21.MaterialAttributes
    BaseColor=LinearInterpolate(A=@n21.BaseColor, B=@n33, Alpha=@n35.R)
    EmissiveColor=LinearInterpolate(A=@n21.EmissiveColor, B=@n44, Alpha=@n35.R)
   ) {attributeSetTypes [#1, #2]})
  LinearInterpolate(
   A=ScalarParameter() {defaultValue=1, group=SuperLevelling_Mask, parameterName=SF_EndValue}
   B=ScalarParameter() {group=SuperLevelling_Mask, parameterName=SF_StartValue}
   Alpha=@n15.G
  )
  n45 = Time()
  OneMinus(None=Frac(None=Multiply(A=@n45) {constB=0.5}))
  OneMinus(None=@n16)
 ]}
