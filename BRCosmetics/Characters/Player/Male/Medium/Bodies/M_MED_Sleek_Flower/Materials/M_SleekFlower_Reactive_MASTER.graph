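# Master character material for the Sleek Flower medium male body, using material attributes with Default Lit and Subsurface Profile shading on skeletal meshes.
# The shared PreFX character function feeds MF_SleekFlower_Reactive, which applies the reactive skin effect using the SkinFX_Mask texture.
# An optional static switch layers MF_Silk, borrowed from weapon skins, through a selectable mask channel.
# The shared PostFX character function produces the final result.
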Material /BRCosmetics/Characters/Player/Male/Medium/Bodies/M_MED_Sleek_Flower/Materials/M_SleekFlower_Reactive_MASTER
graph {
 properties {
  bUseMaterialAttributes=true
  bUsedWithMeshDeformer=true
  bUsedWithMorphTargets=true
  bUsedWithSkeletalMesh=true
  pixelDepthOffsetMode=PDOM_Legacy
  shadingModel=MSM_FromMaterialExpression
  subsurfaceProfile=&SS_HeroSkin_02
  usedShadingModels=Default Lit | Subsurface Profile
 }
 nodes [
  n1 = MaterialFunctionCall() {materialFunction=&BaseFN_CharacterFunction_PreFX}
  n2 = TextureSample(Tex=TextureObjectParameter() {group=Skin FX, parameterName=SkinFX_Mask})
  n3 = ChannelMaskParameter(None=@n2.RGBA) {group=Skin FX, parameterName=Silk Channel}
  n4 = MaterialFunctionCall!(FX_Mask=@n2.RGBA, Material=@n1.Result) {materialFunction=&/MF_SleekFlower_Reactive}
  MaterialFunctionCall(
   AO=@n1.AO
   Distance Blend=@n1.AO
   Fresnel=@n1.AO
   Gmap=@n1.AO
   Pre Skinned Local Position=@n1.AO
   Material=StaticSwitchParameter(
    True=MaterialFunctionCall!(Alpha=@n3.R, Material=@n4.Result) {materialFunction=/Game/Weapons/WeaponSkins/MaterialFunction/MF_Silk}.Result
    False=@n4.Result
   ) {group=Skin FX, parameterName=Use Silk}
  ) {materialFunction=&BaseFN_CharacterFunction_PostFX}
 ]
}
